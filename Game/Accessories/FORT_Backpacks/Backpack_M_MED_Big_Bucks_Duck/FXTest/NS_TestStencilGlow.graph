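# Test Niagara system for a stencil glow effect, filed under the Big Bucks Duck backpack's FXTest folder.
# A single StencilGlow emitter burst-spawns into a mesh renderer with non-uniform scale, a small position offset and randomized material value.
# Particles are not killed when their lifetime elapses and are set to die when the emitter deactivates, so the mesh persists.
# It has no motion modules or user parameters.
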NiagaraSystem /Game/Accessories/FORT_Backpacks/Backpack_M_MED_Big_Bucks_Duck/FXTest/NS_TestStencilGlow
dependencies {usedDataInterfaces [], usedDynamicInputs [], usedModules [&SystemState]}
emitters {StencilGlow {
  emitterUpdateScript [
   &EmitterState(
    Inactive Response=::NewEnumerator1
    Life Cycle Mode=::NewEnumerator1
    Loop Behavior=::NewEnumerator1
    Loop Duration Mode=::NewEnumerator1
    UseLoopDelay=false
   )
   &SpawnBurst_Instantaneous(Use Spawn Probability=false, UseLoopCountLimit=false) {module=SpawnBurst_Instantaneous001}
  ]
  particleSpawnScript [&InitializeParticle(
    Color Mode=::NewEnumerator0
    Lifetime=5
    Mesh Scale=<0.5 0.4 0.5>
    Mesh Scale Mode=::NewEnumerator4
    Position Offset=<0 -10 0>
    UsePositionOffset=true
    Write Material Random=true
   )]
  particleUpdateScript [&ParticleState(Kill Particles When Lifetime Has Elapsed=false, Let Infinitely Lived Particles Die When Emitter Deactivates=true)]
  rendererClasses [&NiagaraMeshRendererProperties]
  renderers [{rendererClass=&NiagaraMeshRendererProperties, rendererIndex=0}]
 }}
summary {
 systemName=NS_TestStencilGlow
 emitters [{emitterName=StencilGlow, rendererClasses [&NiagaraMeshRendererProperties]}]
 userVariables []
}
